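# Physics asset for Bass_PH, apparently a placeholder bass prop in the temporary mesh folder of the StrideMiceDeep group emote in SparksCosmetics.
# A single body on the handle bone uses one capsule, offset along Z and rotated, as a rough collision shape.
# It has no constraints, so it defines no jointed ragdoll or secondary-motion chain.
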PhysicsAsset /SparksCosmetics/Emotes/Group/StrideMiceDeep/SKM_StrideMiceDeep_Temp/Bass_PH_PhysicsAsset
bodies {handle [{shapeType=Capsule, length=64.8404, radius=18.2708, center=<-1.2903 0.5517 39.4105>, rotation=<rot 0.4133 -90 0.5272>}]}
constraints []
