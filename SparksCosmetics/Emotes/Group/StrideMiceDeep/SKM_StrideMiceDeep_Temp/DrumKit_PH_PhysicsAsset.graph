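# Physics asset for a drum kit prop, stored in a temp folder of the StrideMiceDeep group emote.
# Each body is a single capsule on a stand, shake, cymbal, drum or throne bone, plus a tiny root body on the handle.
# Constraints form a tree from the handle through stands to shake, cymbal and seat bones, with toms hanging off the kick drum.
# All constraints share identical 45-degree swing and twist limit values.
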PhysicsAsset /SparksCosmetics/Emotes/Group/StrideMiceDeep/SKM_StrideMiceDeep_Temp/DrumKit_PH_PhysicsAsset
bodies {
 crash_cymbal_l [{shapeType=Capsule, length=17.6096, radius=17.6096, center=<0.1769 -0.0207 -1.1546>, rotation=<rot 2.2104 4.0162 -90>}]
 crash_cymbal_r [{shapeType=Capsule, length=20.7314, radius=20.7314, center=<-0.061 -0.1289 -1.3741>, rotation=<rot 0.3044 -52.2645 -90>}]
 crash_shake_l [{shapeType=Capsule, length=22.0949, radius=4.5836, center=<-2.2214 0.0308 0.1299>, rotation=<rot 0 90.0574 -81.5364>}]
 crash_shake_r [{shapeType=Capsule, length=22.3746, radius=4.6171, center=<2.4371 0.3015 0.9435>, rotation=<rot 0 91.6352 -80.5955>}]
 crash_stand_l [{shapeType=Capsule, length=54.9331, radius=20.0188, center=<55.0598 -2.5786 3.5157>, rotation=<rot 0 89.7753 -89.273>}]
 crash_stand_r [{shapeType=Capsule, length=55.022, radius=20.181, center=<54.3523 0.5526 2.5751>, rotation=<rot 0 -91.1362 -88.49>}]
 floortom_shake [{shapeType=Capsule, length=21.8031, radius=20.6566, center=<20.237 -0.4369 0.6002>, rotation=<rot 2.4281 2.7126 -90>}]
 floortom_stand [{shapeType=Capsule, length=23.9481, radius=22.1723, center=<8.8882 -3.6916 -1.6403>, rotation=<rot 0 -177.2423 -89.1056>}]
 handle [{shapeType=Capsule, length=0.505, radius=0.505, rotation=<rot 0 -90 0>}]
 hihat_bottom_cymbal [{shapeType=Capsule, length=15.9909, radius=15.9907, center=<1.4693 -0.2571 -0.1899>, rotation=<rot 0.0014 -90 -0.3004>}]
 hihat_stand [{shapeType=Capsule, length=54.2585, radius=24.7768, center=<50.6003 0.4748 -2.9724>, rotation=<rot 0 -90.3832 -84.8522>}]
 hihat_top_cymbal [{shapeType=Capsule, length=15.9908, radius=15.9907, center=<-1.2923 -0.2296 -0.2044>, rotation=<rot 0.0016 -90 -0.298>}]
 kickdrum [{shapeType=Capsule, length=57.0028, radius=37.9664, center=<-29.6028 -0.9029 36.3012>, rotation=<rot -0.7393 -90 -22.0363>}]
 ride_cymbal [{shapeType=Capsule, length=25.8639, radius=25.8638, center=<0.1903 -0.0448 -1.8222>, rotation=<rot 1.9393 -34.3957 -90>}]
 ride_shake [{shapeType=Capsule, length=21.0952, radius=6.0105, center=<-1.5552 0.2405 0.7899>, rotation=<rot 0 93.8445 -75.2918>}]
 ride_stand [{shapeType=Capsule, length=42.5696, radius=20.0173, center=<43.0029 -1.601 -2.4526>, rotation=<rot 0 89.8845 -88.9283>}]
 snare_shake [{shapeType=Capsule, length=19.6335, radius=19.1825, center=<14.4837 -0.9927 0.1826>, rotation=<rot 0 1.3676 -169.4164>}]
 snare_stand [{shapeType=Capsule, length=20.4902, radius=18.295, center=<10.1898 0.1935 4.6686>, rotation=<rot 0.5005 -179.151 -90>}]
 throne_seat [{shapeType=Capsule, length=14.2757, radius=14.2745, center=<-9.1576 -0.8971 0.5836>, rotation=<rot 89.6541 0.0015 90.0015>}]
 throne_stand [{shapeType=Capsule, length=20.4793, radius=18.0797, center=<11.9779 3.5817 0.3493>, rotation=<rot 0 89.0804 -179.9332>}]
 tom_shake_l [{shapeType=Capsule, length=23.9232, radius=15.9001, center=<-0.3846 -0.1814 20.6034>, rotation=<rot 0.853 -90 -0.2269>}]
 tom_shake_r [{shapeType=Capsule, length=24.291, radius=18.4233, center=<-0.2523 0.7087 18.6968>, rotation=<rot 1.6516 -90 0.2891>}]
}
constraints [
 {bone1Name=hihat_stand, bone2Name=handle, swing1LimitDegrees=45, swing2LimitDegrees=45, twistLimitDegrees=45}
 {bone1Name=hihat_bottom_cymbal, bone2Name=hihat_stand, swing1LimitDegrees=45, swing2LimitDegrees=45, twistLimitDegrees=45}
 {bone1Name=hihat_top_cymbal, bone2Name=hihat_stand, swing1LimitDegrees=45, swing2LimitDegrees=45, twistLimitDegrees=45}
 {bone1Name=ride_stand, bone2Name=handle, swing1LimitDegrees=45, swing2LimitDegrees=45, twistLimitDegrees=45}
 {bone1Name=ride_shake, bone2Name=ride_stand, swing1LimitDegrees=45, swing2LimitDegrees=45, twistLimitDegrees=45}
 {bone1Name=ride_cymbal, bone2Name=ride_shake, swing1LimitDegrees=45, swing2LimitDegrees=45, twistLimitDegrees=45}
 {bone1Name=crash_stand_r, bone2Name=handle, swing1LimitDegrees=45, swing2LimitDegrees=45, twistLimitDegrees=45}
 {bone1Name=crash_shake_r, bone2Name=crash_stand_r, swing1LimitDegrees=45, swing2LimitDegrees=45, twistLimitDegrees=45}
 {bone1Name=crash_cymbal_r, bone2Name=crash_shake_r, swing1LimitDegrees=45, swing2LimitDegrees=45, twistLimitDegrees=45}
 {bone1Name=crash_stand_l, bone2Name=handle, swing1LimitDegrees=45, swing2LimitDegrees=45, twistLimitDegrees=45}
 {bone1Name=crash_shake_l, bone2Name=crash_stand_l, swing1LimitDegrees=45, swing2LimitDegrees=45, twistLimitDegrees=45}
 {bone1Name=crash_cymbal_l, bone2Name=crash_shake_l, swing1LimitDegrees=45, swing2LimitDegrees=45, twistLimitDegrees=45}
 {bone1Name=kickdrum, bone2Name=handle, swing1LimitDegrees=45, swing2LimitDegrees=45, twistLimitDegrees=45}
 {bone1Name=tom_shake_r, bone2Name=kickdrum, swing1LimitDegrees=45, swing2LimitDegrees=45, twistLimitDegrees=45}
 {bone1Name=tom_shake_l, bone2Name=kickdrum, swing1LimitDegrees=45, swing2LimitDegrees=45, twistLimitDegrees=45}
 {bone1Name=floortom_stand, bone2Name=handle, swing1LimitDegrees=45, swing2LimitDegrees=45, twistLimitDegrees=45}
 {bone1Name=floortom_shake, bone2Name=floortom_stand, swing1LimitDegrees=45, swing2LimitDegrees=45, twistLimitDegrees=45}
 {bone1Name=snare_stand, bone2Name=handle, swing1LimitDegrees=45, swing2LimitDegrees=45, twistLimitDegrees=45}
 {bone1Name=snare_shake, bone2Name=snare_stand, swing1LimitDegrees=45, swing2LimitDegrees=45, twistLimitDegrees=45}
 {bone1Name=throne_stand, bone2Name=handle, swing1LimitDegrees=45, swing2LimitDegrees=45, twistLimitDegrees=45}
 {bone1Name=throne_seat, bone2Name=throne_stand, swing1LimitDegrees=45, swing2LimitDegrees=45, twistLimitDegrees=45}
]
