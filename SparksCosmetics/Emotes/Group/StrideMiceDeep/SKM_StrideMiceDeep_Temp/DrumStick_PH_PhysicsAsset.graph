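# Physics asset for a placeholder drumstick skeletal mesh used by the StrideMiceDeep group emote in SparksCosmetics.
# It has a single body on the handle bone, shaped as one capsule running along the stick's length.
# With no constraints, the drumstick collides or simulates as one rigid piece.
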PhysicsAsset /SparksCosmetics/Emotes/Group/StrideMiceDeep/SKM_StrideMiceDeep_Temp/DrumStick_PH_PhysicsAsset
bodies {handle [{shapeType=Capsule, length=21.9405, radius=1.6972, center=<0.0183 0.0049 14.9376>, rotation=<rot 0.0131 -90 0.049>}]}
constraints []
