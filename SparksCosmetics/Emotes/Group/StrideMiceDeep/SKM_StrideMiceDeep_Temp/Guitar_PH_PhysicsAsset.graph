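# Physics asset for the Guitar_PH guitar mesh, kept in a temp folder under the StrideMiceDeep group emote in SparksCosmetics.
# A single body on the handle bone approximates the guitar with one capsule, offset and rotated from the bone origin.
# It has no constraints, so there is no jointed ragdoll setup.
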PhysicsAsset /SparksCosmetics/Emotes/Group/StrideMiceDeep/SKM_StrideMiceDeep_Temp/Guitar_PH_PhysicsAsset
bodies {handle [{shapeType=Capsule, length=55.7628, radius=18.3736, center=<-1.7373 -0.3542 23.0931>, rotation=<rot -0.2907 -90 -0.0837>}]}
constraints []
