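# Translucent green material in the temporary skeletal mesh folder of the StrideMiceDeep group emote, apparently a placeholder.
# It uses Default Lit shading with surface translucency lighting and is flagged for skeletal mesh and mesh deformer use.
# The graph holds only a constant green color and a scalar constant, with no textures, parameters or visible wiring.
# An Interchange import data reference indicates it came from an asset import.
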Material /SparksCosmetics/Emotes/Group/StrideMiceDeep/SKM_StrideMiceDeep_Temp/Transparent_Green
graph {
 properties {
  assetImportData="<self>:InterchangeAssetImportData_0"
  bUsedWithMeshDeformer=true
  bUsedWithSkeletalMesh=true
  blendMode=BLEND_Translucent
  translucencyLightingMode=TLM_Surface
  usedShadingModels=Default Lit
 }
 nodes [
  Constant3Vector() {constant=<rgba 0 0.8 0 1>}
  Constant() {r=-10872351}
 ]
}
